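# Physics asset for the DolphinGill male medium body cosmetic, with capsule bodies for hanging garment folds and the spine, pelvis and legs.
# Eleven three-segment fold chains ring the body, each starting at a MSTR anchor constrained to spine_02 and/or spine_03.
# Tight angular limits join most segments.
# Cross constraints link neighboring front folds, extending to the side only on the right, and extra thigh capsules hang off the pelvis.
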PhysicsAsset /BRCosmetics/Characters/Player/Male/Medium/Bodies/M_MED_DolphinGill/Meshes/M_MED_DolphinGill_Physics_FE
bodies {
 calf_l [{shapeType=Capsule, length=21.1904, radius=8.3635, center=<-19.6274 -0.826 0.6636>, rotation=<rot 0 -92.0634 -89.8773>}]
 calf_r [{shapeType=Capsule, length=21.1742, radius=8.2036, center=<19.7277 1.2159 -0.649>, rotation=<rot 0 -90.4773 -89.5812>}]
 fold_01_ba_ct [{shapeType=Capsule, length=9.6476, radius=5, shapeName=fold_01_fr_l_capsule, center=<4.9923 -3.0536 0>, rotation=<rot 90 0 0>}]
 fold_01_ba_ct_MSTR [{shapeType=Capsule, length=0.505, radius=0.505, shapeName=fold_01_ba_ct_MSTR_capsule, rotation=<rot 0 -90 0>}]
 fold_01_diag_ba_l [{shapeType=Capsule, length=9.6476, radius=5, shapeName=fold_01_fr_l_capsule, center=<2.6752 -3.1656 0>, rotation=<rot 90 0 0>}]
 fold_01_diag_ba_l_MSTR [{shapeType=Capsule, length=0.505, radius=0.505, shapeName=fold_01_diag_ba_l_MSTR_capsule, rotation=<rot 0 -90 0>}]
 fold_01_diag_ba_r [{shapeType=Capsule, length=9.6476, radius=5, shapeName=fold_01_fr_l_capsule, center=<2.6752 -3.1656 0>, rotation=<rot 90 0 0>}]
 fold_01_diag_ba_r_MSTR [{shapeType=Capsule, length=0.505, radius=0.505, shapeName=fold_01_diag_ba_r_MSTR_capsule, rotation=<rot 0 -90 0>}]
 fold_01_diag_fr_l [{shapeType=Capsule, length=7, radius=5, shapeName=fold_01_fr_l_capsule, center=<4.5483 -3.7698 0>}]
 fold_01_diag_fr_l_MSTR [{shapeType=Capsule, length=0, radius=2.5, shapeName=fold_01_fr_l_MSTR_capsule}]
 fold_01_diag_fr_r [{shapeType=Capsule, length=7, radius=5, shapeName=fold_01_fr_l_capsule, center=<4.5483 -3.7698 0>}]
 fold_01_diag_fr_r_MSTR [{shapeType=Capsule, length=0, radius=2.5, shapeName=fold_01_fr_l_MSTR_capsule}]
 fold_01_fr_l [{shapeType=Capsule, length=1.9928, radius=4.0498, shapeName=fold_01_fr_l_capsule, center=<3.9214 -2.1525 -3.1176>, rotation=<rot 90 0 0>}]
 fold_01_fr_l_MSTR [{shapeType=Capsule, length=0, radius=2.5, shapeName=fold_01_fr_l_MSTR_capsule}]
 fold_01_fr_r [{shapeType=Capsule, length=1.9928, radius=4.0498, shapeName=fold_01_fr_l_capsule, center=<3.9214 -2.1525 3.1176>, rotation=<rot 90 0 0>}]
 fold_01_fr_r_MSTR [{shapeType=Capsule, length=0, radius=2.5, shapeName=fold_01_fr_l_MSTR_capsule}]
 fold_01_leg_ba_l [{shapeType=Capsule, length=9.6476, radius=5, shapeName=fold_01_fr_l_capsule, center=<4.9923 -3.0536 0>, rotation=<rot 90 0 0>}]
 fold_01_leg_ba_l_MSTR [{shapeType=Capsule, length=0.505, radius=0.505, shapeName=fold_01_leg_ba_l_MSTR_capsule, rotation=<rot 0 -90 0>}]
 fold_01_leg_ba_r [{shapeType=Capsule, length=9.6476, radius=5, shapeName=fold_01_fr_l_capsule, center=<4.9923 -4.0564 0>, rotation=<rot 90 0 0>}]
 fold_01_leg_ba_r_MSTR [{shapeType=Capsule, length=0.505, radius=0.505, shapeName=fold_01_leg_ba_r_MSTR_capsule, rotation=<rot 0 -90 0>}]
 fold_01_side_l [{shapeType=Capsule, length=9.6476, radius=5, shapeName=fold_01_fr_l_capsule, center=<4.5483 -3.7698 0>}]
 fold_01_side_l_MSTR [{shapeType=Capsule, length=0, radius=2.5, shapeName=fold_01_fr_l_MSTR_capsule}]
 fold_01_side_r [{shapeType=Capsule, length=9.6476, radius=5, shapeName=fold_01_fr_l_capsule, center=<4.5483 -3.7698 0>}]
 fold_01_side_r_MSTR [{shapeType=Capsule, length=0, radius=2.5, shapeName=fold_01_fr_l_MSTR_capsule}]
 fold_02_ba_ct [{shapeType=Capsule, length=9.6476, radius=5, shapeName=fold_02_fr_l_capsule, center=<4.9923 -3.0536 0>, rotation=<rot 90 0 0>}]
 fold_02_diag_ba_l [{shapeType=Capsule, length=9.6476, radius=5, shapeName=fold_02_fr_l_capsule, center=<2.6752 -3.1656 0>, rotation=<rot 90 0 0>}]
 fold_02_diag_ba_r [{shapeType=Capsule, length=9.6476, radius=5, shapeName=fold_02_fr_l_capsule, center=<2.6752 -3.1656 0>, rotation=<rot 90 0 0>}]
 fold_02_diag_fr_l [{shapeType=Capsule, length=7, radius=5, shapeName=fold_02_fr_l_capsule, center=<4.5483 -3.7698 0>}]
 fold_02_diag_fr_r [{shapeType=Capsule, length=7, radius=5, shapeName=fold_02_fr_l_capsule, center=<4.5483 -3.7698 0>}]
 fold_02_fr_l [{shapeType=Capsule, length=6.3978, radius=4.4119, shapeName=fold_02_fr_l_capsule, center=<3.9214 -3.7302 -3.392>, rotation=<rot 90 0 0>}]
 fold_02_fr_r [{shapeType=Capsule, length=6.3978, radius=4.4119, shapeName=fold_02_fr_l_capsule, center=<3.9214 -4.2356 3.392>, rotation=<rot 90 0 0>}]
 fold_02_leg_ba_l [{shapeType=Capsule, length=9.6476, radius=5, shapeName=fold_02_fr_l_capsule, center=<4.9923 -3.0536 0>, rotation=<rot 90 0 0>}]
 fold_02_leg_ba_r [{shapeType=Capsule, length=9.6476, radius=5, shapeName=fold_02_fr_l_capsule, center=<4.9923 -3.991 0>, rotation=<rot 90 0 0>}]
 fold_02_side_l [{shapeType=Capsule, length=9.6476, radius=5, shapeName=fold_02_fr_l_capsule, center=<4.5483 -3.7698 0>}]
 fold_02_side_r [{shapeType=Capsule, length=9.6476, radius=5, shapeName=fold_02_fr_l_capsule, center=<4.5483 -3.7698 0>}]
 fold_03_ba_ct [{shapeType=Capsule, length=9.6476, radius=5, shapeName=fold_03_fr_l_capsule, center=<4.9923 -3.188 0>, rotation=<rot 90 0 0>}]
 fold_03_diag_ba_l [{shapeType=Capsule, length=9.6476, radius=5, shapeName=fold_03_fr_l_capsule, center=<3.8169 -4.2494 0>, rotation=<rot 90 0 0>}]
 fold_03_diag_ba_r [{shapeType=Capsule, length=9.6476, radius=5, shapeName=fold_03_fr_l_capsule, center=<3.8169 -4.2494 0>, rotation=<rot 90 0 0>}]
 fold_03_diag_fr_l [
  {shapeType=Capsule, length=7, radius=5, shapeName=fold_03_fr_l_capsule, center=<4.5483 -3.7698 0>}
  {shapeType=Capsule, length=8.5, radius=4.5, shapeName=fold_03_diag_fr_l_capsule, center=<9.9052 -3.2484 0>}
 ]
 fold_03_diag_fr_r [
  {shapeType=Capsule, length=7, radius=5, shapeName=fold_03_fr_l_capsule, center=<4.5483 -3.7698 0>}
  {shapeType=Capsule, length=8.5, radius=3.5, shapeName=fold_03_diag_fr_r_capsule, center=<10.1379 -3.2549 0>}
 ]
 fold_03_fr_l [
  {shapeType=Capsule, length=8.6684, radius=4.4119, shapeName=fold_03_fr_l_capsule, center=<5.9929 -4.0494 -2.5156>, rotation=<rot 90 0 0>}
  {
   shapeType=Capsule
   length=10.1684
   radius=4.4119
   shapeName=fold_03_fr_l_capsule
   center=<11.0681 -3.7524 -4.5987>
   rotation=<rot 0 -0.0001 -20>
  }
 ]
 fold_03_fr_r [{shapeType=Capsule, length=8.6684, radius=4.4119, shapeName=fold_03_fr_l_capsule, center=<5.9929 -4.2971 2.5156>, rotation=<rot 90 0 0>}]
 fold_03_leg_ba_l [{shapeType=Capsule, length=9.6476, radius=5, shapeName=fold_03_fr_l_capsule, center=<4.9923 -6.0687 0>, rotation=<rot 90 0 0>}]
 fold_03_leg_ba_r [{shapeType=Capsule, length=9.6476, radius=5, shapeName=fold_03_fr_l_capsule, center=<4.9923 -3.9283 0>, rotation=<rot 90 0 0>}]
 fold_03_side_l [{shapeType=Capsule, length=8.1476, radius=6.5, shapeName=fold_03_fr_l_capsule, center=<7.0607 -5.3983 0>}]
 fold_03_side_r [{shapeType=Capsule, length=8.1476, radius=6.5, shapeName=fold_03_fr_l_capsule, center=<7.5125 -5.3983 0>}]
 foot_l [{shapeType=Capsule, length=15.0236, radius=8.8005, center=<-1.53 6.7288 0.2049>, rotation=<rot 0 5.0885 -87.8862>}]
 foot_r [{shapeType=Capsule, length=15.0059, radius=8.639, center=<1.7957 -6.7622 -0.3348>, rotation=<rot 0 5.5399 -87.5803>}]
 pelvis [
  {shapeType=Capsule, length=2.4555, radius=10.5864, shapeName=pelvis_capsule, center=<0 -1.4688 0>, rotation=<rot 0 -90 0>}
  {shapeType=Capsule, length=14.0683, radius=11.9317, shapeName=pelvis_capsule, center=<1.8465 0 0>, rotation=<rot 90 0 0>}
  {shapeType=Capsule, length=7.3531, radius=10, shapeName=pelvis_capsule, center=<-7.4887 -2.4499 0>}
 ]
 spine_01 [{shapeType=Capsule, length=6.8331, radius=10.3297, shapeName=pelvis_capsule, center=<0 -1.3623 0>, rotation=<rot 0 -90 0>}]
 spine_02 [{shapeType=Capsule, length=6.1621, radius=9.6501, shapeName=pelvis_capsule, center=<0 -1.7909 0>, rotation=<rot 0 -90 0>}]
 spine_03 [{shapeType=Capsule, length=7.0017, radius=9.5658, shapeName=pelvis_capsule, rotation=<rot 0 -90 0>}]
 thigh_l [{
   shapeType=Capsule
   length=34.9215
   radius=7.9604
   shapeName=thigh_l_capsule
   center=<-18.8337 0.1987 0.2904>
   rotation=<rot 89.0311 2.8133 0.0013>
  }]
 thigh_l_bodies_01 [{
   shapeType=Capsule
   length=17.6273
   radius=10.7527
   shapeName=thigh_l_capsule
   center=<-15.9142 0.4936 8.3408>
   rotation=<rot 88.0414 -44.5973 -39.079>
  }]
 thigh_r [{
   shapeType=Capsule
   length=34.9215
   radius=7.9604
   shapeName=thigh_r_capsule
   center=<18.8427 -0.3711 -0.2575>
   rotation=<rot 89.0311 2.8133 -179.9987>
  }]
 thigh_r_bodies_01 [{
   shapeType=Capsule
   length=17.6273
   radius=10.7527
   shapeName=thigh_r_capsule
   center=<15.9142 -0.4936 -8.3408>
   rotation=<rot 88.0414 -44.5973 140.921>
  }]
}
constraints [
 {bone1Name=spine_01, bone2Name=pelvis, swing1LimitDegrees=45, swing2LimitDegrees=45, twistLimitDegrees=45}
 {bone1Name=spine_02, bone2Name=spine_01, swing1LimitDegrees=45, swing2LimitDegrees=45, twistLimitDegrees=45}
 {bone1Name=thigh_l, bone2Name=pelvis, swing1LimitDegrees=45, swing2LimitDegrees=45, twistLimitDegrees=45}
 {bone1Name=calf_l, bone2Name=thigh_l, swing1LimitDegrees=45, swing2LimitDegrees=45, twistLimitDegrees=45}
 {bone1Name=foot_l, bone2Name=calf_l, swing1LimitDegrees=45, swing2LimitDegrees=45, twistLimitDegrees=45}
 {bone1Name=thigh_r, bone2Name=pelvis, swing1LimitDegrees=45, swing2LimitDegrees=45, twistLimitDegrees=45}
 {bone1Name=calf_r, bone2Name=thigh_r, swing1LimitDegrees=45, swing2LimitDegrees=45, twistLimitDegrees=45}
 {bone1Name=foot_r, bone2Name=calf_r, swing1LimitDegrees=45, swing2LimitDegrees=45, twistLimitDegrees=45}
 {
  bone1Name=spine_03
  bone2Name=spine_02
  swing1LimitDegrees=45
  swing1Motion=Free
  swing2LimitDegrees=45
  swing2Motion=Free
  twistLimitDegrees=45
  twistMotion=Free
 }
 {
  bone1Name=fold_01_fr_l_MSTR
  bone2Name=spine_02
  swing1LimitDegrees=45
  swing1Motion=Free
  swing2LimitDegrees=45
  swing2Motion=Free
  twistLimitDegrees=45
  twistMotion=Free
 }
 {bone1Name=fold_01_fr_l, bone2Name=fold_01_fr_l_MSTR, swing1LimitDegrees=40, swing2LimitDegrees=30, twistLimitDegrees=40}
 {bone1Name=fold_02_fr_l, bone2Name=fold_01_fr_l, swing1LimitDegrees=20, swing2LimitDegrees=15, twistLimitDegrees=20}
 {bone1Name=fold_03_fr_l, bone2Name=fold_02_fr_l, swing1LimitDegrees=45, swing2LimitDegrees=15, twistLimitDegrees=20}
 {
  bone1Name=fold_01_fr_r_MSTR
  bone2Name=spine_02
  swing1LimitDegrees=45
  swing1Motion=Free
  swing2LimitDegrees=45
  swing2Motion=Free
  twistLimitDegrees=45
  twistMotion=Free
 }
 {bone1Name=fold_01_fr_r, bone2Name=fold_01_fr_r_MSTR, swing1LimitDegrees=20, swing2LimitDegrees=15, twistLimitDegrees=5}
 {bone1Name=fold_02_fr_r, bone2Name=fold_01_fr_r, swing1LimitDegrees=20, swing2LimitDegrees=15, twistLimitDegrees=5}
 {bone1Name=fold_03_fr_r, bone2Name=fold_02_fr_r, swing1LimitDegrees=20, swing2LimitDegrees=15, twistLimitDegrees=5}
 {
  bone1Name=fold_01_side_l_MSTR
  bone2Name=spine_02
  swing1LimitDegrees=45
  swing1Motion=Free
  swing2LimitDegrees=45
  swing2Motion=Free
  twistLimitDegrees=45
  twistMotion=Free
 }
 {bone1Name=fold_01_side_l, bone2Name=fold_01_side_l_MSTR, swing1LimitDegrees=20, swing2LimitDegrees=15, twistLimitDegrees=5}
 {bone1Name=fold_02_side_l, bone2Name=fold_01_side_l, swing1LimitDegrees=20, swing2LimitDegrees=15, twistLimitDegrees=5}
 {bone1Name=fold_03_side_l, bone2Name=fold_02_side_l, swing1LimitDegrees=20, swing2LimitDegrees=15, twistLimitDegrees=5}
 {
  bone1Name=fold_01_side_r_MSTR
  bone2Name=spine_02
  swing1LimitDegrees=45
  swing1Motion=Free
  swing2LimitDegrees=45
  swing2Motion=Free
  twistLimitDegrees=45
  twistMotion=Free
 }
 {bone1Name=fold_01_side_r, bone2Name=fold_01_side_r_MSTR, swing1LimitDegrees=20, swing2LimitDegrees=15, twistLimitDegrees=1}
 {bone1Name=fold_02_side_r, bone2Name=fold_01_side_r, swing1LimitDegrees=20, swing2LimitDegrees=15, twistLimitDegrees=1}
 {bone1Name=fold_03_side_r, bone2Name=fold_02_side_r, swing1LimitDegrees=20, swing2LimitDegrees=15, twistLimitDegrees=1}
 {bone1Name=fold_01_diag_fr_l, bone2Name=fold_01_fr_l, swing1LimitDegrees=20, swing2LimitDegrees=25, twistLimitDegrees=25}
 {
  bone1Name=fold_03_diag_fr_l
  bone2Name=fold_03_fr_l
  swing1LimitDegrees=20
  swing1Motion=Free
  swing2LimitDegrees=25
  swing2Motion=Free
  twistLimitDegrees=25
  twistMotion=Free
 }
 {
  bone1Name=fold_01_fr_r_MSTR
  bone2Name=spine_03
  swing1LimitDegrees=45
  swing1Motion=Locked
  swing2LimitDegrees=45
  swing2Motion=Locked
  twistLimitDegrees=45
  twistMotion=Locked
 }
 {bone1Name=fold_01_diag_fr_r, bone2Name=fold_01_fr_r, swing1LimitDegrees=35, swing2LimitDegrees=25, twistLimitDegrees=25}
 {bone1Name=fold_02_diag_fr_r, bone2Name=fold_02_fr_r, swing1LimitDegrees=35, swing2LimitDegrees=25, twistLimitDegrees=25}
 {bone1Name=fold_03_diag_fr_r, bone2Name=fold_03_fr_r, swing1LimitDegrees=35, swing2LimitDegrees=25, twistLimitDegrees=25}
 {
  bone1Name=fold_01_diag_fr_l_MSTR
  bone2Name=spine_03
  swing1LimitDegrees=45
  swing1Motion=Locked
  swing2LimitDegrees=45
  swing2Motion=Locked
  twistLimitDegrees=45
  twistMotion=Locked
 }
 {bone1Name=fold_01_diag_fr_l, bone2Name=fold_01_diag_fr_l_MSTR, swing1LimitDegrees=20, swing2LimitDegrees=15, twistLimitDegrees=5}
 {bone1Name=fold_02_diag_fr_l, bone2Name=fold_01_diag_fr_l, swing1LimitDegrees=20, swing2LimitDegrees=15, twistLimitDegrees=5}
 {bone1Name=fold_03_diag_fr_l, bone2Name=fold_02_diag_fr_l, swing1LimitDegrees=45, swing2LimitDegrees=15, twistLimitDegrees=5}
 {
  bone1Name=fold_01_diag_fr_r_MSTR
  bone2Name=spine_03
  swing1LimitDegrees=45
  swing1Motion=Locked
  swing2LimitDegrees=45
  swing2Motion=Locked
  twistLimitDegrees=45
  twistMotion=Locked
 }
 {bone1Name=fold_01_diag_fr_r, bone2Name=fold_01_diag_fr_r_MSTR, swing1LimitDegrees=20, swing2LimitDegrees=15, twistLimitDegrees=5}
 {bone1Name=fold_02_diag_fr_r, bone2Name=fold_01_diag_fr_r, swing1LimitDegrees=20, swing2LimitDegrees=15, twistLimitDegrees=5}
 {bone1Name=fold_01_side_r, bone2Name=fold_01_diag_fr_r, swing1LimitDegrees=35, swing2LimitDegrees=25, twistLimitDegrees=25}
 {bone1Name=fold_03_diag_fr_r, bone2Name=fold_02_diag_fr_r, swing1LimitDegrees=20, swing2LimitDegrees=15, twistLimitDegrees=5}
 {bone1Name=fold_02_side_r, bone2Name=fold_02_diag_fr_r, swing1LimitDegrees=35, swing2LimitDegrees=25, twistLimitDegrees=25}
 {bone1Name=fold_03_side_r, bone2Name=fold_03_diag_fr_r, swing1LimitDegrees=35, swing2LimitDegrees=25, twistLimitDegrees=5}
 {
  bone1Name=fold_01_side_l_MSTR
  bone2Name=spine_03
  swing1LimitDegrees=45
  swing1Motion=Locked
  swing2LimitDegrees=45
  swing2Motion=Locked
  twistLimitDegrees=45
  twistMotion=Locked
 }
 {
  bone1Name=fold_01_diag_ba_l_MSTR
  bone2Name=spine_02
  swing1LimitDegrees=45
  swing1Motion=Free
  swing2LimitDegrees=45
  swing2Motion=Free
  twistLimitDegrees=45
  twistMotion=Free
 }
 {bone1Name=fold_03_diag_ba_l, bone2Name=fold_02_diag_ba_l, swing1LimitDegrees=20, swing2LimitDegrees=10, twistLimitDegrees=1}
 {
  bone1Name=fold_01_diag_ba_r_MSTR
  bone2Name=spine_02
  swing1LimitDegrees=45
  swing1Motion=Free
  swing2LimitDegrees=45
  swing2Motion=Free
  twistLimitDegrees=45
  twistMotion=Free
 }
 {bone1Name=fold_03_diag_ba_r, bone2Name=fold_02_diag_ba_r, swing1LimitDegrees=20, swing2LimitDegrees=10, twistLimitDegrees=1}
 {bone1Name=fold_02_diag_ba_l, bone2Name=fold_01_diag_ba_l, swing1LimitDegrees=20, swing2LimitDegrees=10, twistLimitDegrees=1}
 {bone1Name=fold_02_diag_ba_r, bone2Name=fold_01_diag_ba_r, swing1LimitDegrees=20, swing2LimitDegrees=10, twistLimitDegrees=1}
 {bone1Name=fold_03_ba_ct, bone2Name=fold_02_ba_ct, swing1LimitDegrees=20, swing2LimitDegrees=1, twistLimitDegrees=2.5}
 {bone1Name=fold_03_leg_ba_r, bone2Name=fold_02_leg_ba_r, swing1LimitDegrees=20, swing2LimitDegrees=5, twistLimitDegrees=2.5}
 {bone1Name=fold_03_leg_ba_l, bone2Name=fold_02_leg_ba_l, swing1LimitDegrees=20, swing2LimitDegrees=5, twistLimitDegrees=2.5}
 {bone1Name=fold_02_leg_ba_l, bone2Name=fold_01_leg_ba_l, swing1LimitDegrees=20, swing2LimitDegrees=5, twistLimitDegrees=2.5}
 {bone1Name=fold_02_leg_ba_r, bone2Name=fold_01_leg_ba_r, swing1LimitDegrees=20, swing2LimitDegrees=5, twistLimitDegrees=2.5}
 {bone1Name=fold_02_ba_ct, bone2Name=fold_01_ba_ct, swing1LimitDegrees=20, swing2LimitDegrees=1, twistLimitDegrees=2.5}
 {bone1Name=fold_01_ba_ct, bone2Name=fold_01_ba_ct_MSTR, swing1LimitDegrees=15, swing2LimitDegrees=1, twistLimitDegrees=2.5}
 {bone1Name=fold_01_leg_ba_l, bone2Name=fold_01_leg_ba_l_MSTR, swing1LimitDegrees=15, swing2LimitDegrees=5, twistLimitDegrees=2.5}
 {bone1Name=fold_01_leg_ba_r, bone2Name=fold_01_leg_ba_r_MSTR, swing1LimitDegrees=15, swing2LimitDegrees=5, twistLimitDegrees=2.5}
 {bone1Name=fold_01_diag_ba_r, bone2Name=fold_01_diag_ba_r_MSTR, swing1LimitDegrees=20, swing2LimitDegrees=10, twistLimitDegrees=1}
 {bone1Name=fold_01_diag_ba_l, bone2Name=fold_01_diag_ba_l_MSTR, swing1LimitDegrees=20, swing2LimitDegrees=10, twistLimitDegrees=1}
 {
  bone1Name=fold_01_leg_ba_l_MSTR
  bone2Name=spine_02
  swing1LimitDegrees=45
  swing1Motion=Free
  swing2LimitDegrees=45
  swing2Motion=Free
  twistLimitDegrees=45
  twistMotion=Free
 }
 {
  bone1Name=fold_01_ba_ct_MSTR
  bone2Name=spine_02
  swing1LimitDegrees=45
  swing1Motion=Free
  swing2LimitDegrees=45
  swing2Motion=Free
  twistLimitDegrees=45
  twistMotion=Free
 }
 {
  bone1Name=fold_01_leg_ba_r_MSTR
  bone2Name=spine_02
  swing1LimitDegrees=45
  swing1Motion=Free
  swing2LimitDegrees=45
  swing2Motion=Free
  twistLimitDegrees=45
  twistMotion=Free
 }
 {
  bone1Name=thigh_l_bodies_01
  bone2Name=pelvis
  swing1LimitDegrees=45
  swing1Motion=Free
  swing2LimitDegrees=45
  swing2Motion=Free
  twistLimitDegrees=45
  twistMotion=Free
 }
 {
  bone1Name=thigh_r_bodies_01
  bone2Name=pelvis
  swing1LimitDegrees=45
  swing1Motion=Free
  swing2LimitDegrees=45
  swing2Motion=Free
  twistLimitDegrees=45
  twistMotion=Free
 }
]
