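# Animation Blueprint for the DolphinGill male medium cosmetic body, copying the attached parent mesh's pose and layering procedural motion on it.
# A pre-physics Control Rig feeds two spine_02-space RigidBody nodes toggled between front-end and gameplay/emote physics assets, then a post Control Rig.
# The update event derives rig and toggle inputs from crouch, movement, turning, DBNO, melee, emote and front-end states.
# It also computes a gravity override, which the AnimGraph leaves unused.
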AnimBlueprint /BRCosmetics/Characters/Player/Male/Medium/Bodies/M_MED_DolphinGill/Meshes/M_MED_DolphinGill_AnimBP
target_skeleton=&/M_MED_DolphinGill_Skeleton
graphs [
 AnimGraph {nodes [
   n1 = Control Rig!(
    Source=Copy Pose From Mesh() {bUseAttachedParent=true}
    Alpha
    crouching:double=Get crouching_f() -> (crouching_f:double)
    dbno:double=Get dbno_f() -> (dbno_f:double)
   ) -> (Pose) {controlRigAssetReference {blueprintRigClass=&/M_MED_DolphinGill_CtrlRig_01.M_MED_DolphinGill_CtrlRig_01_C}}
   n2 = RigidBody!(
    ComponentPose=Local To Component(LocalPose=@n1)
    bAlphaBoolEnabled:bool=Get NOT_FE() -> (NOT_FE:bool)
    OverrideWorldGravity:Vector="0,0,-980"
    SimSpaceSettings:Sim Space Settings Structure=(WorldAlpha=1,VelocityScaleZ=0.3,DampingAlpha=1,MaxLinearVelocity=3000,MaxAngularVelocity=1500,MaxLinearAcceleration=1000,MaxAngularAcceleration=65,ExternalLinearDragV=(X=1,Y=1,Z=1),ExternalLinearVelocity=(X=0,Y=0,Z=0),ExternalAngularVelocity=(X=0,Y=0,Z=0))
   ) -> (Pose) {
    alphaInputType=Bool
    bClampLinearTranslationLimitToRefPose=true
    bOverrideWorldGravity=true
    overridePhysicsAsset=&/M_MED_DolphinGill_Physics_07
    simulationSpace=BaseBoneSpace
    alphaBoolBlend {blendInTime=0.2, blendOption=Linear, blendOutTime=0.2}
    baseBoneRef {boneName=spine_02}
    overrideWorldGravity {z=-980}
    simSpaceSettings {
     maxAngularAcceleration=65
     maxAngularVelocity=1500
     maxLinearAcceleration=1000
     maxLinearVelocity=3000
     velocityScaleZ=0.3
     worldAlpha=1
     externalLinearDragV=<1 1 1>
    }
   }
   n3 = RigidBody(
    ComponentPose=@n2
    bAlphaBoolEnabled=Get FE() -> (FE:bool)
    OverrideWorldGravity="0,0,-980"
    SimSpaceSettings=(WorldAlpha=1,VelocityScaleZ=0.3,DampingAlpha=1,MaxLinearVelocity=3000,MaxAngularVelocity=1500,MaxLinearAcceleration=1000,MaxAngularAcceleration=65,ExternalLinearDragV=(X=1,Y=1,Z=1),ExternalLinearVelocity=(X=0,Y=0,Z=0),ExternalAngularVelocity=(X=0,Y=0,Z=0))
   ) {
    alphaInputType=Bool
    bClampLinearTranslationLimitToRefPose=true
    bOverrideWorldGravity=true
    overridePhysicsAsset=&/M_MED_DolphinGill_Physics_FE
    simulationSpace=BaseBoneSpace
    alphaBoolBlend {blendInTime=0.2, blendOption=Linear, blendOutTime=0.2}
    baseBoneRef {boneName=spine_02}
    overrideWorldGravity {z=-980}
    simSpaceSettings {
     maxAngularAcceleration=65
     maxAngularVelocity=1500
     maxLinearAcceleration=1000
     maxLinearVelocity=3000
     velocityScaleZ=0.3
     worldAlpha=1
     externalLinearDragV=<1 1 1>
    }
   }
   n4 = Control Rig!(
    Source=Component To Local(ComponentPose=@n3)
    Alpha
    crouching=Get crouching_f()
    dbno=Get dbno_f()
    crouching_or_dbno:double=Get crouching_or_dbno_f() -> (crouching_or_dbno_f:double)
    PlayingMeleeAnim:double=Get PlayingMeleeAnim_f() -> (PlayingMeleeAnim_f:double)
   ) -> (Pose) {controlRigAssetReference {blueprintRigClass=&/M_MED_DolphinGill_CtrlRig_post.M_MED_DolphinGill_CtrlRig_post_C}}
   Output Pose(Result=@n4) {layerGroup=None}
  ]}
 EventGraph {nodes [
   Event BlueprintUpdateAnimation()
   > n1 = Sequence!(execute) -> (then_0, then_1, then_2:Exec)
   n2 = Get bIsMoving2D()
   n3 = "Absolute (Float)"(A=BreakRotator(InRot=Get PawnDeltaRotation() -> (PawnDeltaRotation:Rotator)).Yaw)
   > Set static_crouching_rotation(static_crouching_rotation:bool=AND Boolean!(A="float > float"(A=@n3, B=2), B=Get bIsCrouching(), C:bool=NOT Boolean(A=@n2)) -> (ReturnValue)) -> (Output_Get:bool)
   n4 = AND Boolean!(A=Get bIsCrouching(), B=Get bIsMoving2D()) -> (ReturnValue)
   n5 = Get static_crouching_rotation() -> (static_crouching_rotation:bool)
   > Set crouching_f(crouching_f:double="To Float (Boolean)"(InBool=OR Boolean(A=@n4, B=@n5))) -> (Output_Get:double)
   n6 = Get bIsDBNO()
   > Set dbno_f(dbno_f:double="To Float (Boolean)"(InBool=NOT Boolean(A=@n6))) -> (Output_Get:double)
   n7 = Get bIsCrouching()
   n8 = Get bIsDBNO()
   > Set crouching_or_dbno_f(crouching_or_dbno_f:double="To Float (Boolean)"(InBool=OR Boolean(A=@n7, B=@n8))) -> (Output_Get:double)
   > Set PlayingMeleeAnim_f(PlayingMeleeAnim_f:double="To Float (Boolean)"(InBool=Get bIsPlayingMeleeAnim())) -> (Output_Get:double)
   n9 = Get bIsPlayingEmote()
   > Set FE(FE:bool=AND Boolean(A=Get bIsInFrontEnd(), B=NOT Boolean(A=@n9))) -> (Output_Get:bool)
   n10 = Get bIsInFrontEnd()
   > Set NOT_FE(NOT_FE:bool=OR Boolean(A=NOT Boolean(A=@n10), B=Get bIsPlayingEmote())) -> (Output_Get:bool)
   n11 = "Gravity Override State Function (Parameter Asset) (Nativized)"(
    execute=@n1.then_1
    CharacterPartAnimInstance=Self-Reference()
    GameTimeInSeconds=GetGameTimeInSeconds()
    Parameters=&/M_MED_DolphinGill_Parameters
   )
   > Set Gravity Override(Gravity Override=@n11)
   Event BlueprintBeginPlay() {is=title K2Node_Event}
   > Set play(play:bool=true) -> (Output_Get:bool)
  ]}
]
variables [
 crouching_f {type=number}
 crouching_or_dbno_f {type=number}
 dbno_f {type=number}
 FE {type=boolean}
 Gravity Override {type=Vector}
 NOT_FE {type=boolean}
 play {type=boolean}
 PlayingMeleeAnim_f {type=number}
 Sim Space Settings {
  type=SimSpaceSettings
  default {
   dampingAlpha=1
   maxAngularAcceleration=10000
   maxAngularVelocity=10000
   maxLinearAcceleration=10000
   maxLinearVelocity=10000
   velocityScaleZ=1
   worldAlpha=0
   externalLinearDragV=<0 0 0>
  }
 }
 static_crouching_rotation {type=boolean}
]
